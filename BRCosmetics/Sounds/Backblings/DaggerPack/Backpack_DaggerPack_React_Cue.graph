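# Sound cue for the DaggerPack back bling's reaction sound, a cosmetic backpack effect.
# A Random node picks one of three equally weighted React wave variations.
# It plays at reduced volume through the back bling sound class, using a close-range pet attenuation and a reactive back bling audio property sheet.
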SoundCue /BRCosmetics/Sounds/Backblings/DaggerPack/Backpack_DaggerPack_React_Cue
root=Random_0
nodes [
 {id=Random_0, children [WavePlayer_0,WavePlayer_1,WavePlayer_5], properties {weights [1,1,1]}}
 {id=WavePlayer_0, properties {soundWaveAssetPtr=&/Backbling_DaggerPack_React_03}}
 {id=WavePlayer_1, properties {soundWaveAssetPtr=&/Backbling_DaggerPack_React_01}}
 {id=WavePlayer_5, properties {soundWaveAssetPtr=&/Backbling_DaggerPack_React_02}}
]
properties {
 attenuationSettings=&Fort_Pets_Close_Athena_Att
 audioPropertiesSheet=/BRCosmetics/Sounds/Settings/PropertySheets/Reactive/APS_Cosmetic_Reactive_Backbling
 maxDistance=1800.0001
 soundClassObject=/Game/Sounds/FortSoundClassesAndModes/Fort_Backblings
 volumeMultiplier=0.63
}
